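# Sound cue for distant SprocketPoppy creature groans, classed as an outdoor ambient one-shot.
# A Random node picks one of fourteen equally weighted groan waves, excluded from branch culling, and feeds it through a Modulator for variation.
# A distant-creature attenuation asset shapes falloff, and a distance-based send feeds the exterior ambient source bus.
# Overridden concurrency caps simultaneous groans at two, scaling volume by distance.
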SoundCue /Hermes/Audio/SprocketPoppyDistant/Spot_SprocketPoppy_Groan_Distant_Cue
root=Modulator_5
nodes [
 {id=Modulator_5, children [Random_6]}
 {
  id=Random_6
  children [
   WavePlayer_45
   WavePlayer_46
   WavePlayer_47
   WavePlayer_48
   WavePlayer_49
   WavePlayer_50
   WavePlayer_51
   WavePlayer_52
   WavePlayer_53
   WavePlayer_54
   WavePlayer_55
   WavePlayer_56
   WavePlayer_57
   WavePlayer_58
  ]
  properties {bShouldExcludeFromBranchCulling=true, weights [1,1,1,1,1,1,1,1,1,1,1,1,1,1]}
 }
 {id=WavePlayer_45, properties {soundWaveAssetPtr=&/Spot_SprocketPoppy_Groan_Distant_01}}
 {id=WavePlayer_46, properties {soundWaveAssetPtr=&/Spot_SprocketPoppy_Groan_Distant_02}}
 {id=WavePlayer_47, properties {soundWaveAssetPtr=&/Spot_SprocketPoppy_Groan_Distant_03}}
 {id=WavePlayer_48, properties {soundWaveAssetPtr=&/Spot_SprocketPoppy_Groan_Distant_04}}
 {id=WavePlayer_49, properties {soundWaveAssetPtr=&/Spot_SprocketPoppy_Groan_Distant_05}}
 {id=WavePlayer_50, properties {soundWaveAssetPtr=&/Spot_SprocketPoppy_Groan_Distant_06}}
 {id=WavePlayer_51, properties {soundWaveAssetPtr=&/Spot_SprocketPoppy_Groan_Distant_07}}
 {id=WavePlayer_52, properties {soundWaveAssetPtr=&/Spot_SprocketPoppy_Groan_Distant_08}}
 {id=WavePlayer_53, properties {soundWaveAssetPtr=&/Spot_SprocketPoppy_Groan_Distant_09}}
 {id=WavePlayer_54, properties {soundWaveAssetPtr=&/Spot_SprocketPoppy_Groan_Distant_10}}
 {id=WavePlayer_55, properties {soundWaveAssetPtr=&/Spot_SprocketPoppy_Groan_Distant_11}}
 {id=WavePlayer_56, properties {soundWaveAssetPtr=&/Spot_SprocketPoppy_Groan_Distant_12}}
 {id=WavePlayer_57, properties {soundWaveAssetPtr=&/Spot_SprocketPoppy_Groan_Distant_13}}
 {id=WavePlayer_58, properties {soundWaveAssetPtr=&/Spot_SprocketPoppy_Groan_Distant_14}}
]
properties {
 attenuationSettings=/AmbientAudioBase/Attenuations/Fort_OneShot_Distant_Creature_Att
 bOverrideConcurrency=true
 maxDistance=8500
 soundClassObject=&Fort_Ambient_OneShots_Outdoors
 volumeMultiplier=1
 busSends [{
   audioBus=null
   maxSendDistance=8000
   minSendDistance=1000
   minSendLevel=0.4
   sendLevel=1
   soundSourceBus=&AmbientAudio_Exterior_SourceBus
  }]
 concurrencyOverrides {maxCount=2, voiceStealReleaseTime=0.25, volumeScaleMode=Distance}
}
